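# Control Rig for the Buzz Cool NPC head that builds FK face controls procedurally from the skeleton.
# Construction clears arrays, imports bones, gathers descendants of the faceAttach bone, and spawns a control per joint under its parent's namesake.
# Name matching gives eye (non-lid) and jaw joints separate control settings, with eyes colored by side.
# Forward solve offsets the faceAttach control to its bone, then copies control transforms onto face joints by array index.
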ControlRigBlueprint /BRCosmetics/Characters/Player/Male/Medium/Bodies/M_MED_BuzzCool_NPC/Meshes/Parts/M_MED_BuzzCool_NPCHead_CtrlRig
graphs [
 New Function_ContainedGraph {nodes [
   n1 = RigVMFunctionEntryNode()
   RigVMFunction_Sequence = &RigVMFunction_Sequence(ExecuteContext=@n1.ExecuteContext)
   n2 = &RigUnit_HierarchyImportFromSkeleton(NameSpace=None, bIncludeCurves=False, bIncludeMeshSockets=False, bIncludeVirtualBones=True, ExecutePin=@RigVMFunction_Sequence.A)
   > RigVMVariableNode(Value=@n2.Items) {variable=Skeleton Array}
   > For_Each = RigVMDispatchNode(Array=RigVMVariableNode() {variable=Skeleton Array}.Value) {template="DISPATCH_RigVMDispatch_ArrayIterator(in Array,out Element,out Index,out Count,out Ratio)"}
   Contains = /Script/RigVM.RigVMFunction_Contains(Search=faceAttach, Name=@For_Each.Element.Name) {is=struct RigVMUnitNode, template="Contains::Execute(in Name,in Search,out Result)"}
   RigVMFunction_ControlFlowBranch = &RigVMFunction_ControlFlowBranch(Condition=@Contains.Result, ExecuteContext=@For_Each.ExecuteContext)
   RigVMVariableNode(
    Value=/Script/ControlRig.RigUnit_CollectionChildrenArray(TypeToSearch=Bone, bDefaultChildren=True, bIncludeParent=False, bRecursive=true, Parent=@For_Each.Element) {is=struct RigVMUnitNode}.Items
    ExecuteContext=@RigVMFunction_ControlFlowBranch.True
   ) {variable=Face Joints}
   RigVMFunctionReturnNode(ExecuteContext=@RigVMFunction_Sequence.B)
  ]}
 New Function_ContainedGraph {nodes [
   n1 = RigVMFunctionEntryNode()
   RigVMFunction_Sequence = &RigVMFunction_Sequence(ExecuteContext=@n1.ExecuteContext)
   RigVMFunctionReturnNode(ExecuteContext=@RigVMFunction_Sequence.B)
   RigVMDispatchNode(Array=RigVMVariableNode() {variable=Skeleton Array}.Value, ExecuteContext=@RigVMFunction_Sequence.A) {template="DISPATCH_RigVMDispatch_ArrayReset(io Array)"}
   > RigVMDispatchNode(Array=RigVMVariableNode() {variable=Face Joints}.Value) {template="DISPATCH_RigVMDispatch_ArrayReset(io Array)"}
   > RigVMDispatchNode(Array=RigVMVariableNode() {variable=Control Array}.Value) {template="DISPATCH_RigVMDispatch_ArrayReset(io Array)"}
  ]}
 New Function_ContainedGraph {nodes [
   n1 = RigVMFunctionEntryNode()
   RigVMFunction_Sequence = &RigVMFunction_Sequence(ExecuteContext=@n1.ExecuteContext)
   RigVMFunctionReturnNode(ExecuteContext=@RigVMFunction_Sequence.B)
   /Script/ControlRig.RigUnit_HierarchyAddControlTransform(
    InitialValue=(Rotation=(X=0,Y=0,Z=0,W=1),Translation=(X=0,Y=0,Z=0),Scale3D=(X=1,Y=1,Z=1))
    Name=faceAttach
    OffsetSpace=LocalSpace
    Parent=(Type=Bone,Name="None")
    Settings="(InitialSpace=LocalSpace,bUsePreferredRotationOrder=False,PreferredRotationOrder=YZX,Limits=(LimitTranslationX=(bMinimum=False,bMaximum=False),LimitTranslationY=(bMinimum=False,bMaximum=False),LimitTranslationZ=(bMinimum=False,bMaximum=False),LimitPitch=(b ...(truncated)"
    OffsetTransform=&RigUnit_GetTransform(Item=(Type=Bone,Name="faceAttach"), bInitial=False).Transform
    ExecutePin=@RigVMFunction_Sequence.A
   ) {
    is=struct RigVMUnitNode
    template="SpawnControl::Execute(in InitialValue,in Settings,in OffsetTransform,in OffsetSpace,in Parent,in Name,out Item)"
   }
   > For_Each = RigVMDispatchNode(Array=RigVMVariableNode() {variable=Face Joints}.Value) {template="DISPATCH_RigVMDispatch_ArrayIterator(in Array,out Element,out Index,out Count,out Ratio)"}
   > RigVMVariableNode(Value=@For_Each.Element) {variable=Element}
   > RigVMVariableNode(Value=@For_Each.Element.Name) {variable=Element Name}
   > n2 = RigVMVariableNode() {variable=Control Created}
   n3 = &RigVMFunction_Sequence(ExecuteContext=@n2.ExecuteContext)
   n4 = &RigVMFunction_Sequence(ExecuteContext=@n3.A)
   n5 = RigVMVariableNode() {variable=Element Name}
   Contains = &RigVMFunction_Contains(Search=eye, Name=@n5.Value) {template="Contains::Execute(in Name,in Search,out Result)"}
   RigVMFunction_MathBoolNot = &RigVMFunction_MathBoolNot(Value=&RigVMFunction_Contains(Search=lid, Name=@n5.Value) {template="Contains::Execute(in Name,in Search,out Result)"}.Result)
   RigVMFunction_MathBoolAnd = &RigVMFunction_MathBoolAnd(A=@Contains.Result, B=@RigVMFunction_MathBoolNot.Result)
   RigVMFunction_ControlFlowBranch = &RigVMFunction_ControlFlowBranch(Condition=@RigVMFunction_MathBoolAnd.Result, ExecuteContext=@n4.A)
   n6 = RigVMVariableNode() {variable=Element Name}
   n7 = &RigUnit_HierarchyAddControlTransform(
    InitialValue=(Rotation=(X=0,Y=0,Z=0,W=1),Translation=(X=0,Y=0,Z=0),Scale3D=(X=1,Y=1,Z=1))
    OffsetSpace=GlobalSpace
    Parent=(Type=Control,Name="None")
    Settings="(InitialSpace=LocalSpace,bUsePreferredRotationOrder=False,PreferredRotationOrder=YZX,Limits=(LimitTranslationX=(bMinimum=False,bMaximum=False),LimitTranslationY=(bMinimum=False,bMaximum=False),LimitTranslationZ=(bMinimum=False,bMaximum=False),LimitPitch=(b ...(truncated)"
    OffsetTransform=&RigUnit_GetTransform(bInitial=False, Item=RigVMVariableNode() {variable=Element}.Value).Transform
    Parent.Name=/Script/ControlRig.RigUnit_HierarchyGetParent(bDefaultParent=True, Child=RigVMVariableNode() {variable=Element}.Value) {is=struct RigVMUnitNode}.Parent.Name
    Settings.Shape.Color=RigVMDispatchNode(
     False=(R=0,G=0,B=1,A=1)
     True=(R=1,G=0,B=0,A=1)
     Condition=&RigVMFunction_Contains(Search=R_, Name=@n6.Value) {template="Contains::Execute(in Name,in Search,out Result)"}.Result
    ) {template="DISPATCH_RigVMDispatch_If(in Condition,in True,in False,out Result)"}.Result
    Name=RigVMVariableNode() {variable=Element Name}.Value
    ExecutePin=@RigVMFunction_ControlFlowBranch.True
   ) {template="SpawnControl::Execute(in InitialValue,in Settings,in OffsetTransform,in OffsetSpace,in Parent,in Name,out Item)"}
   > RigVMDispatchNode(Element=@n7.Item, Array=RigVMVariableNode() {variable=Control Array}.Value) {template="DISPATCH_RigVMDispatch_ArrayAdd(io Array,in Element,out Index)"}
   > RigVMVariableNode(Value=true) {variable=Control Created}
   n8 = &RigVMFunction_ControlFlowBranch(
    Condition=&RigVMFunction_Contains(Search=jaw, Name=RigVMVariableNode() {variable=Element Name}.Value) {template="Contains::Execute(in Name,in Search,out Result)"}.Result
    ExecuteContext=@n4.B
   )
   n9 = &RigUnit_HierarchyAddControlTransform(
    InitialValue=(Rotation=(X=0,Y=0,Z=0,W=1),Translation=(X=0,Y=0,Z=0),Scale3D=(X=1,Y=1,Z=1))
    OffsetSpace=GlobalSpace
    Parent=(Type=Control,Name="None")
    Settings="(InitialSpace=LocalSpace,bUsePreferredRotationOrder=False,PreferredRotationOrder=YZX,Limits=(LimitTranslationX=(bMinimum=False,bMaximum=False),LimitTranslationY=(bMinimum=False,bMaximum=False),LimitTranslationZ=(bMinimum=False,bMaximum=False),LimitPitch=(b ...(truncated)"
    OffsetTransform=&RigUnit_GetTransform(bInitial=False, Item=RigVMVariableNode() {variable=Element}.Value).Transform
    Parent.Name=&RigUnit_HierarchyGetParent(bDefaultParent=True, Child=RigVMVariableNode() {variable=Element}.Value).Parent.Name
    Name=RigVMVariableNode() {variable=Element Name}.Value
    ExecutePin=@n8.True
   ) {template="SpawnControl::Execute(in InitialValue,in Settings,in OffsetTransform,in OffsetSpace,in Parent,in Name,out Item)"}
   > RigVMDispatchNode(Element=@n9.Item, Array=RigVMVariableNode() {variable=Control Array}.Value) {template="DISPATCH_RigVMDispatch_ArrayAdd(io Array,in Element,out Index)"}
   > RigVMVariableNode(Value=true) {variable=Control Created}
   n10 = &RigVMFunction_ControlFlowBranch(Condition=RigVMVariableNode() {variable=Control Created}.Value, ExecuteContext=@n3.B)
   SpawnControl = &RigUnit_HierarchyAddControlTransform(
    InitialValue=(Rotation=(X=0,Y=0,Z=0,W=1),Translation=(X=0,Y=0,Z=0),Scale3D=(X=1,Y=1,Z=1))
    OffsetSpace=GlobalSpace
    Parent=(Type=Control,Name="None")
    Settings="(InitialSpace=LocalSpace,bUsePreferredRotationOrder=False,PreferredRotationOrder=YZX,Limits=(LimitTranslationX=(bMinimum=False,bMaximum=False),LimitTranslationY=(bMinimum=False,bMaximum=False),LimitTranslationZ=(bMinimum=False,bMaximum=False),LimitPitch=(b ...(truncated)"
    OffsetTransform=&RigUnit_GetTransform(bInitial=False, Item=RigVMVariableNode() {variable=Element}.Value).Transform
    Parent.Name=&RigUnit_HierarchyGetParent(bDefaultParent=True, Child=RigVMVariableNode() {variable=Element}.Value).Parent.Name
    Name=RigVMVariableNode() {variable=Element Name}.Value
    ExecutePin=@n10.False
   ) {template="SpawnControl::Execute(in InitialValue,in Settings,in OffsetTransform,in OffsetSpace,in Parent,in Name,out Item)"}
   > RigVMDispatchNode(Element=@SpawnControl.Item, Array=RigVMVariableNode() {variable=Control Array}.Value) {template="DISPATCH_RigVMDispatch_ArrayAdd(io Array,in Element,out Index)"}
  ]}
 New Function_ContainedGraph {nodes [
   n1 = RigVMFunctionEntryNode()
   RigVMFunction_Sequence = &RigVMFunction_Sequence(ExecuteContext=@n1.ExecuteContext)
   SetControlOffset = /Script/ControlRig.RigUnit_SetControlOffset(
    Control=faceAttach
    Space=LocalSpace
    Offset=&RigUnit_GetTransform(Item=(Type=Bone,Name="faceAttach"), bInitial=False).Transform
    ExecutePin=@RigVMFunction_Sequence.A
   ) {is=struct RigVMUnitNode, template="SetControlOffset::Execute(in Control,io Offset,in Space)"}
   > For_Each_1 = RigVMDispatchNode(Array=RigVMVariableNode() {variable=Control Array}.Value) {template="DISPATCH_RigVMDispatch_ArrayIterator(in Array,out Element,out Index,out Count,out Ratio)"}
   > RigVMVariableNode(Value=@For_Each_1.Element) {variable=Element}
   > RigVMVariableNode(Value=@For_Each_1.Element.Name) {variable=Name}
   > RigVMVariableNode(Value=@For_Each_1.Index) {variable=Index}
   > &RigUnit_SetTransform(
    Weight=1
    bInitial=False
    Item=RigVMDispatchNode(Array=RigVMVariableNode() {variable=Face Joints}.Value, Index=RigVMVariableNode() {variable=Index}.Value) {template="DISPATCH_RigVMDispatch_ArrayGetAtIndex(in Array,in Index,out Element)"}.Element
    Value=&RigUnit_GetTransform(bInitial=False, Item=RigVMVariableNode() {variable=Element}.Value).Transform
   )
   RigVMFunctionReturnNode(ExecuteContext=@RigVMFunction_Sequence.B)
  ]}
 RigVMFunctionLibrary {nodes [
   Create Controls = RigVMCollapseNode(ExecuteContext=())
   FK Solve = RigVMCollapseNode(ExecuteContext=())
   Import Skeleton = RigVMCollapseNode(ExecuteContext=())
   Reset Arrays = RigVMCollapseNode(ExecuteContext=())
  ]}
 RigVMModel {nodes [
   RigUnit_BeginExecution = &RigUnit_BeginExecution()
   FK Solve = RigVMFunctionReferenceNode(ExecuteContext=@RigUnit_BeginExecution.ExecutePin)
  ]}
 RigVMModel Construction Graph {nodes [
   n1 = &RigUnit_PrepareForExecution()
   Reset Arrays = RigVMFunctionReferenceNode(ExecuteContext=@n1.ExecutePin)
   > Import Skeleton = RigVMFunctionReferenceNode()
   > Create Controls = RigVMFunctionReferenceNode()
  ]}
]
variables [
 Control Array {type="TArray<FRigElementKey>"}
 Face Joints {type="TArray<FRigElementKey>"}
 Skeleton Array {type="TArray<FRigElementKey>"}
]
